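# Material function for the Dire wheel cosmetic that applies an animated mask to the tire's emissive input.
# It samples the FX Tire Mask texture parameter, a shared gradient mask by default, with stretched UVs panning along U.
# The mask is clamped to a small minimum so the glow never drops to zero, then multiplied into the Emission input.
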MaterialFunction /VehicleCosmetics/Wheels/Dire/Materials/MF_FX_Wheel_DIre_Tire_EM
graph {nodes [
  n1 = TextureSampleParameter2D(UVs=Panner(Coordinate=TextureCoordinate() {uTiling=0.25}) {speedX=0.2}) {group=1_Input Textures, parameterName=FX Tire Mask, sortPriority=51, texture=/VehicleCosmetics/Wheels/_SharedTextures/FX/GradMask_T}
  FunctionOutput(None=Multiply(A=FunctionInput() {inputName=Emission, previewValue=<1 1 1 1>}, B=Clamp(None=@n1.RGB) {minDefault=0.05}))
 ]}
