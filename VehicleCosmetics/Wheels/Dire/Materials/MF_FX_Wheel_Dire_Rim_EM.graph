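# Material function that builds the emissive output for the Dire wheel rim, with animated noise in masked areas.
# Two samples of one noise texture parameter pan in opposite horizontal directions at different tilings, then get multiplied together, boosted, limited by the mask's green channel and clamped.
# That result is added to the incoming Emission, which is multiplied by the clamped sum of the mask's red and blue channels.
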MaterialFunction /VehicleCosmetics/Wheels/Dire/Materials/MF_FX_Wheel_Dire_Rim_EM
graph {nodes [
  n1 = TextureObjectParameter() {group=1_Input Textures, parameterName=FX Noise, sortPriority=50, texture=/VehicleCosmetics/Wheels/_SharedTextures/FX/ContrastNoiseGradient_T}
  n2 = TextureSample(UVs=Panner(Coordinate=TextureCoordinate() {vTiling=0.5}) {speedX=0.1, speedY=-0.1}, Tex=@n1) {samplerSource=SSM_Wrap_WorldGroupSettings, texture=&ContrastNoiseGradient_T}
  n3 = TextureSample(UVs=Panner(Coordinate=TextureCoordinate() {uTiling=2, vTiling=0.5}) {speedX=-0.1, speedY=-0.11}, Tex=@n1) {samplerSource=SSM_Wrap_WorldGroupSettings, texture=&ContrastNoiseGradient_T}
  n4 = FunctionInput() {bUsePreviewValueAsDefault=true, inputName=Masks In, previewValue=<1 1 1 1>}
  n5 = Multiply(A=Multiply(A=Multiply(A=@n2.RGB, B=@n3.RGB)) {constB=4}, B=ComponentMask(None=@n4) {g=true})
  n6 = FunctionInput() {inputName=Emission, sortPriority=1, previewValue=<1 1 1 1>}
  n7 = Clamp(None=Add(A=ComponentMask(None=@n4) {r=true}, B=ComponentMask(None=@n4) {b=true}))
  FunctionOutput(None=Add(A=Clamp(None=@n5), B=Multiply(A=@n6, B=@n7)))
 ]}
