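# Translucent unlit material for the glow effect on the Dire cosmetic wheel, usable on skeletal meshes.
# Two counter-rotating samples of one glow texture multiply into a moving pattern, layered over its static green channel and modulated by two slow, stretched panning samples.
# Back Brightness sets that result's exponent, Glow Brightness boosts the rotating pattern, and their sum plus inverted Opacity is tinted by Emissive Color.
# Vertex red separately masks the shaped result.
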Material /VehicleCosmetics/Wheels/Dire/Materials/M_FX_Wheel_Dire_Glow
graph {
 properties {
  bAutomaticallySetUsageInEditor=false
  bUsedWithMeshDeformer=true
  bUsedWithSkeletalMesh=true
  blendMode=BLEND_Translucent
  shadingModel=MSM_Unlit
  usedShadingModels=Unlit
 }
 nodes [
  n1 = TextureObjectParameter() {group=Base, parameterName=Glow, sortPriority=0, texture=/VehicleCosmetics/Wheels/_SharedTextures/FX/T_Glow}
  n2 = Multiply(
   A=TextureSample(UVs=Rotator(), Tex=@n1) {samplerSource=SSM_Wrap_WorldGroupSettings, samplerType=SAMPLERTYPE_LinearColor}.R
   B=TextureSample(UVs=Rotator() {speed=-0.15}, Tex=@n1) {samplerSource=SSM_Wrap_WorldGroupSettings, samplerType=SAMPLERTYPE_LinearColor}.R
  )
  n3 = Add(A=Add(A=TextureSample(Tex=@n1) {samplerSource=SSM_Wrap_WorldGroupSettings, samplerType=SAMPLERTYPE_LinearColor}.G, B=@n2), B=@n2)
  n4 = TextureCoordinate() {uTiling=0.01, vTiling=0.01}
  n5 = TextureCoordinate() {uTiling=0.01, vTiling=0.01}
  n6 = Multiply(
   A=TextureSample(UVs=Panner(Coordinate=@n4) {speedX=-0.005}, Tex=@n1) {samplerSource=SSM_Wrap_WorldGroupSettings, samplerType=SAMPLERTYPE_LinearColor}.RGB
   B=TextureSample(UVs=Panner(Coordinate=@n5) {speedX=0.005, speedY=0.01}, Tex=@n1) {samplerSource=SSM_Wrap_WorldGroupSettings, samplerType=SAMPLERTYPE_LinearColor}.RGB
  )
  n7 = Power(
   Base=Multiply(A=Clamp(None=@n3), B=Clamp(None=@n6) {minDefault=0.35})
   Exp=ScalarParameter() {defaultValue=1, group=Base, parameterName=Back Brightness, sortPriority=2}
  )
  Multiply(A=VertexColor().R, B=@n7)
  n8 = ScalarParameter() {defaultValue=1, group=Base, parameterName=Opacity, sortPriority=4}
  n9 = Multiply(A=@n2, B=ScalarParameter() {defaultValue=6, group=Base, parameterName=Glow Brightness, sortPriority=3})
  Multiply(
   A=Multiply(A=VectorParameter() {group=Base, parameterName=Emissive Color, sortPriority=1, defaultValue=<rgba 1 1 1 1>}.RGB) {constB=2}
   B=Add(A=OneMinus(None=@n8), B=Add(A=@n7, B=@n9))
  ) {constB=2}
 ]
}
