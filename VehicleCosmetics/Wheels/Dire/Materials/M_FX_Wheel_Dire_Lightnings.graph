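# Unlit translucent Niagara material that renders animated lightning for the Dire wheel cosmetic effect, usable on sprites, ribbons and mesh particles.
# A rotating fire-noise normal map distorts a panning water mask and two lightning samples, one V-flipped, multiplied with a panning grid texture.
# Vertex and particle color plus a Particle Color parameter tint the result.
# A dynamic parameter exposes detail motion, brightness, texture rotation and texture switching per particle.
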Material /VehicleCosmetics/Wheels/Dire/Materials/M_FX_Wheel_Dire_Lightnings
graph {
 properties {
  bAutomaticallySetUsageInEditor=false
  bUsedWithNiagaraMeshParticles=true
  bUsedWithNiagaraRibbons=true
  bUsedWithNiagaraSprites=true
  blendMode=BLEND_Translucent
  shadingModel=MSM_Unlit
  usedShadingModels=Unlit
 }
 nodes [
  n1 = VertexColor()
  n2 = TextureCoordinate() {uTiling=0.25, vTiling=0.25}
  n3 = LinearInterpolate(A=Panner(Coordinate=@n2) {speedX=0.25, speedY=-0.25}, B=Panner(Coordinate=@n2) {speedX=0.25, speedY=-0.15}, Alpha=NamedRerouteUsage())
  n4 = TextureCoordinate() {uTiling=0.25, vTiling=0.25}
  n5 = Multiply(A=Time()) {constB=0.5}
  n6 = ComponentMask(None=TextureSample(UVs=Rotator(Coordinate=@n4, Time=@n5) {speed=-0.2}) {samplerType=SAMPLERTYPE_Normal, texture=/VehicleCosmetics/Wheels/_SharedTextures/FX/Noise_Fire_01_N}.RGB) {g=true, r=true}
  n7 = Panner(Coordinate=Add(A=TextureCoordinate() {uTiling=2.5, vTiling=2.5}, B=Multiply(A=@n6) {constB=0.05})) {speedX=0.1, speedY=0.5}
  n8 = Add(A=Multiply(A=TextureSample(UVs=@n7) {texture=/VehicleCosmetics/Wheels/_SharedTextures/FX/Water_01_Pack}.RGB) {constB=5.5}) {constB=0.1}
  n9 = Multiply(A=@n6) {constB=0.025}
  n10 = Add(A=@n9, B=TextureCoordinate())
  n11 = Constant() {r=0.5}
  n12 = NamedRerouteUsage()
  n13 = TextureSample(UVs=LinearInterpolate(A=Rotator(Coordinate=@n10, Time=@n11) {speed=3.14}, B=@n10, Alpha=@n12)) {texture=/VehicleCosmetics/Wheels/_SharedTextures/FX/LightningFlip_T}
  n14 = Add(A=@n9, B=TextureCoordinate() {vTiling=-1})
  n15 = Constant() {r=0.5}
  n16 = TextureSample(UVs=LinearInterpolate(A=Rotator(Coordinate=@n14, Time=@n15) {speed=3.14}, B=@n14, Alpha=@n12)) {texture=&LightningFlip_T}
  n17 = Multiply(
   A=Multiply(A=TextureSample(UVs=@n3) {texture=/VehicleCosmetics/Wheels/_SharedTextures/FX/WidgetGridVertexColorMaterial}.RGB, B=Clamp(None=@n8))
   B=LinearInterpolate(A=Multiply(A=@n13.RGB) {constB=2}, B=Multiply(A=@n16.RGB) {constB=10}, Alpha=NamedRerouteUsage())
  )
  n18 = Clamp(None=Multiply(A=@n1, B=Clamp(None=@n17)))
  n19 = NamedRerouteUsage()
  Multiply(
   A=VectorParameter() {group=Base, parameterName=Particle Color, sortPriority=-1, defaultValue=<rgba 1 1 1 1>}.RGB
   B=Multiply(A=ParticleColor().RGB, B=Multiply(A=@n18, B=@n19))
  )
  n20 = DynamicParameter() {paramNames [DetailMotion, Brightness, RotateTexture, TextureSwitch]}
  NamedRerouteDeclaration(None=@n20.DetailMotion) {name=DetailMotion, nodeColor=<rgba 0.2656 1 0 1>}
  NamedRerouteDeclaration(None=@n20.Brightness) {name=Brightness, nodeColor=<rgba 0.2656 1 0 1>}
  NamedRerouteDeclaration(None=@n20.RotateTexture) {name=RotateTexture, nodeColor=<rgba 0.2656 1 0 1>}
  NamedRerouteDeclaration(None=@n20.TextureSwitch) {name=TextureSwitch, nodeColor=<rgba 0.2656 1 0 1>}
  Reroute(None=Clamp(None=Multiply(A=@n1.A, B=@n17)))
 ]
}
