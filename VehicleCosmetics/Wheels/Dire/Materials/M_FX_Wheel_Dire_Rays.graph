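# Translucent unlit Niagara material for sprites, ribbons and mesh particles, producing a radial ray effect for the Dire wheel.
# UV distance from center forms a soft circular falloff that masks a noise texture sampled by angle, turning it into rays.
# A time-driven sine varies the noise exponent so the rays pulse.
# A static switch optionally adds a tighter central glow, and a vertex color input is present.
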Material /VehicleCosmetics/Wheels/Dire/Materials/M_FX_Wheel_Dire_Rays
graph {
 properties {
  bAutomaticallySetUsageInEditor=false
  bUsedWithNiagaraMeshParticles=true
  bUsedWithNiagaraRibbons=true
  bUsedWithNiagaraSprites=true
  blendMode=BLEND_Translucent
  shadingModel=MSM_Unlit
  usedShadingModels=Unlit
 }
 nodes [
  n1 = Add(A=TextureCoordinate(), B=Constant() {r=-0.5})
  n2 = SquareRoot(None=DotProduct(A=@n1, B=@n1))
  n3 = Clamp(None=Power(Base=OneMinus(None=@n2), Exp=ScalarParameter() {defaultValue=12, group=Base, parameterName=Power, sortPriority=2}))
  n4 = Power(Base=@n3, Exp=ScalarParameter() {defaultValue=5, group=Base, parameterName=Inner Circle Size, sortPriority=3})
  n5 = Add(
   A=Multiply(A=MaterialFunctionCall() {materialFunction=&VectorToRadialValue}.Vector Converted to Angle)
   B=AppendVector(A=Constant(), B=Constant() {r=1})
  )
  n6 = Sine(None=Multiply(A=Time())) {period=4}
  n7 = Multiply(
   A=@n3
   B=Power(
    Base=TextureSampleParameter2D(UVs=@n5) {group=Base, parameterName=Rays Noise, sortPriority=0, texture=/VehicleCosmetics/Wheels/_SharedTextures/FX/NoiseClouds}.RGB
    Exp=ConstantBiasScale(None=@n6) {bias=0.01, scale=0.75}
   )
  )
  Clamp(None=StaticSwitchParameter(True=Add(A=@n4, B=@n7), False=@n7) {defaultValue=true, group=Base, parameterName=CenterGradient, sortPriority=1})
  VertexColor()
 ]
}
