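# Translucent unlit material for the smoke effect of the Dire vehicle wheel, enabled for Niagara sprites, ribbons and mesh particles.
# An Albedo texture, defaulting to a shared flame texture, is tinted by vertex color, particle color and a Particle Color parameter.
# A Greyscale static switch uses only the texture's red channel for color.
# Opacity multiplies vertex alpha, texture alpha and an intensity parameter, without reading particle alpha.
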Material /VehicleCosmetics/Wheels/Dire/Materials/M_FX_Wheel_Dire_Smoke
graph {
 properties {
  bAutomaticallySetUsageInEditor=false
  bUsedWithNiagaraMeshParticles=true
  bUsedWithNiagaraRibbons=true
  bUsedWithNiagaraSprites=true
  blendMode=BLEND_Translucent
  shadingModel=MSM_Unlit
  usedShadingModels=Unlit
 }
 nodes [
  n1 = VertexColor()
  n2 = TextureSampleParameter2D() {group=Base, parameterName=Albedo, sortPriority=0, texture=/VehicleCosmetics/Wheels/_SharedTextures/FX/Flame_T}
  n3 = Multiply(A=@n1, B=@n2.R)
  n4 = Multiply(A=@n2.RGB, B=@n1)
  Multiply(
   A=VectorParameter() {group=Base, parameterName=Particle Color, sortPriority=-1, defaultValue=<rgba 1 1 1 1>}.RGB
   B=Multiply(A=ParticleColor().RGB, B=StaticSwitchParameter(True=@n3, False=@n4) {group=Base, parameterName=Greyscale, sortPriority=1})
  )
  Reroute(None=Multiply(A=Multiply(A=@n1.A, B=@n2.A), B=ScalarParameter() {defaultValue=1, group=Base, parameterName=Opacity Intensity, sortPriority=2}))
 ]
}
