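# Simplified parent material for the Dire vehicle wheel, built on material attributes for skeletal mesh use.
# MF_Wheel_Dire_Parent_Simple supplies the base attributes, and the alpha of the LOD rim base color texture overrides Opacity Mask.
# The result feeds the shared vehicle damage function, then the vehicle de-rez transition function.
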Material /VehicleCosmetics/Wheels/Dire/Materials/M_Wheel_Dire_Parent_Simple
graph {
 properties {
  bAutomaticallySetUsageInEditor=false
  bUseMaterialAttributes=true
  bUsedWithMeshDeformer=true
  bUsedWithSkeletalMesh=true
  pixelDepthOffsetMode=PDOM_Legacy
  usedShadingModels=Default Lit
 }
 nodes [
  n1 = MaterialFunctionCall() {materialFunction=&/MF_Wheel_Dire_Parent_Simple}
  n2 = TextureSample() {texture=/VehicleCosmetics/Wheels/Dire/Textures/T_Dire_Rim_BC_LOD}
  MaterialFunctionCall(In=MaterialFunctionCall(Material=SetMaterialAttributes!(MaterialAttributes=@n1.MaterialAttributes, Opacity Mask=@n2.A) {attributeSetTypes [#1]}) {materialFunction=&MF_BaseVehicle_Damage}.Result) {materialFunction=&MF_VehicleDeRez_Transition}
 ]
}
